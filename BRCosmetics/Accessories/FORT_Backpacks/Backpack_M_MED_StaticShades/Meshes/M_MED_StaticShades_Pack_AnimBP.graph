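# Animation Blueprint for the StaticShades backpack cosmetic, posing the pack and its dangling parts on the player's body.
# Each update copies pelvis and spine world transforms from the body mesh and derives correction alphas from spine bend and movement states.
# Additive bone offsets adjust pack and cloth for jumping, skydiving and directional movement, including a brief pulse when rightward movement starts.
# Anim Dynamics swings keys, grenade, buckles and lower cloth, whose gravity reverses during skydiving.
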AnimBlueprint /BRCosmetics/Accessories/FORT_Backpacks/Backpack_M_MED_StaticShades/Meshes/M_MED_StaticShades_Pack_AnimBP
target_skeleton=&/M_MED_StaticShades_Pack_Skeleton
graphs [
 AnimGraph {nodes [
   "Transform (Modify) Bone - Bone: dyn_lower_mstr"(Alpha=Get Right movmnt and melee corr() -> (Right movmnt and melee corr:double), Translation="0,-1,0", Rotation="0,-5,0", Scale="1,1,1") {
    rotationMode=BMM_Additive
    rotationSpace=BCS_BoneSpace
    translationMode=BMM_Additive
    translationSpace=BCS_BoneSpace
    alphaBoolBlend {blendInTime=0.2, blendOption=Custom, blendOutTime=3}
    alphaScaleBiasClamp {bInterpResult=true, interpSpeedDecreasing=0.3, interpSpeedIncreasing=3, inRange {max=0.3}}
    boneToModify {boneName=dyn_lower_mstr}
    rotation {yaw=-5}
    translation {y=-1}
   }
   n1 = Get Start moving right() -> (Start moving right:double)
   n2 = Get Wiggle alpha() -> (Wiggle alpha:double)
   n3 = "Transform (Modify) Bone - Bone: root"(ComponentPose=Local To Component(LocalPose=Input Pose()), Translation="0,0.5,3.5", Rotation="0,0,0") {
    translationMode=BMM_Additive
    translationSpace=BCS_BoneSpace
    alphaBoolBlend {blendOption=Linear}
    boneToModify {boneName=root}
    translation {y=0.5, z=3.5}
   }
   n4 = "Constraint - dyn_lower (P)"(ComponentPose=@n3, Alpha=1, ConstraintWeights_0=1) {
    alphaBoolBlend {blendOption=Linear}
    boneToModify {boneName=dyn_lower}
    constraintSetup [{targetBone {boneName=pelvis}}]
    constraintWeights [1]
   }
   n5 = "Transform (Modify) Bone - Bone: pack"(ComponentPose=@n4, Alpha=Get Pack alpha() -> (Pack alpha:double), Translation="0,0,1.5", Rotation="0,0,25") {
    rotationMode=BMM_Additive
    rotationSpace=BCS_BoneSpace
    translationMode=BMM_Additive
    translationSpace=BCS_BoneSpace
    alphaBoolBlend {blendOption=Linear}
    alphaScaleBiasClamp {bInterpResult=true, interpSpeedDecreasing=0.3, interpSpeedIncreasing=3}
    boneToModify {boneName=pack}
    rotation {roll=25}
    translation {z=1.5}
   }
   n6 = "Transform (Modify) Bone - Bone: dyn_lower_mstr"(ComponentPose=@n5, Alpha=Get Cloth alpha() -> (Cloth alpha:double), Translation="0,0,0", Rotation="0,-20,0", Scale="1,1,1") {
    rotationMode=BMM_Additive
    rotationSpace=BCS_BoneSpace
    translationMode=BMM_Additive
    translationSpace=BCS_BoneSpace
    alphaBoolBlend {blendOption=Linear}
    alphaScaleBiasClamp {bInterpResult=true, interpSpeedDecreasing=25}
    boneToModify {boneName=dyn_lower_mstr}
    rotation {yaw=-20}
   }
   n7 = "Transform (Modify) Bone - Bone: pack"(ComponentPose=@n6, Alpha="Get  jump alpha pack"() -> (" jump alpha pack":double), Translation="0,0,0", Rotation="0,0,8") {
    rotationMode=BMM_Additive
    rotationSpace=BCS_BoneSpace
    translationMode=BMM_Additive
    translationSpace=BCS_BoneSpace
    alphaBoolBlend {blendInTime=0.1, blendOption=Custom, blendOutTime=0.6}
    alphaScaleBiasClamp {bInterpResult=true}
    boneToModify {boneName=pack}
    rotation {roll=8}
   }
   n8 = "Transform (Modify) Bone - Bone: dyn_lower_mstr"!(
    ComponentPose=@n7
    Alpha=Get Jump alpha cloth() -> (Jump alpha cloth:double)
    Translation="-0.5,-1,0"
    Rotation_Roll:float=0
    Rotation_Pitch:float=Get Jump cloth Y rot corr() -> (Jump cloth Y rot corr:double)
    Rotation_Yaw:float=Get Jump cloth Z rot corr() -> (Jump cloth Z rot corr:double)
    Scale="1,1,1"
   ) -> (Pose) {
    rotationMode=BMM_Additive
    rotationSpace=BCS_BoneSpace
    translationMode=BMM_Additive
    translationSpace=BCS_BoneSpace
    alphaBoolBlend {blendInTime=0.4, blendOption=QuadraticInOut, blendOutTime=0.6}
    alphaScaleBiasClamp {bInterpResult=true, interpSpeedDecreasing=20, interpSpeedIncreasing=1}
    boneToModify {boneName=dyn_lower_mstr}
    rotation {yaw=-80}
    translation=<-0.5 -1>
   }
   n9 = "Transform (Modify) Bone - Bone: pelvis"(
    ComponentPose=@n8
    Alpha=1
    Translation=Get pelvis_translation() -> (pelvis_translation:Vector)
    Rotation=Get pelvis_rotation() -> (pelvis_rotation:Rotator)
    Scale="1,1,1"
   ) {
    rotationMode=BMM_Replace
    rotationSpace=BCS_WorldSpace
    scaleMode=BMM_Additive
    translationMode=BMM_Replace
    translationSpace=BCS_WorldSpace
    alphaBoolBlend {blendInTime=1, blendOption=Linear, blendOutTime=1}
    boneToModify {boneName=pelvis}
    rotation {roll=20}
   }
   n10 = "Transform (Modify) Bone - Bone: spine_01"(
    ComponentPose=@n9
    Alpha=1
    Translation=Get spine01_translation() -> (spine01_translation:Vector)
    Rotation=Get spine01_rotation() -> (spine01_rotation:Rotator)
    Scale="1,1,1"
   ) {
    rotationMode=BMM_Replace
    rotationSpace=BCS_WorldSpace
    scaleMode=BMM_Additive
    translationMode=BMM_Replace
    translationSpace=BCS_WorldSpace
    alphaBoolBlend {blendInTime=1, blendOption=Linear, blendOutTime=1}
    boneToModify {boneName=spine_01}
    rotation {roll=20}
   }
   n11 = "Transform (Modify) Bone - Bone: spine_02"(
    ComponentPose=@n10
    Alpha=1
    Translation=Get spine02_translation() -> (spine02_translation:Vector)
    Rotation=Get spine02_rotation() -> (spine02_rotation:Rotator)
    Scale="1,1,1"
   ) {
    rotationMode=BMM_Replace
    rotationSpace=BCS_WorldSpace
    scaleMode=BMM_Additive
    translationMode=BMM_Replace
    translationSpace=BCS_WorldSpace
    alphaBoolBlend {blendInTime=1, blendOption=Linear, blendOutTime=1}
    boneToModify {boneName=spine_02}
    rotation {roll=20}
   }
   n12 = "Transform (Modify) Bone - Bone: spine_05"(
    ComponentPose=@n11
    Alpha=1
    Translation=Get spine 05 translation() -> (spine 05 translation:Vector)
    Rotation=Get spine 05 rotation() -> (spine 05 rotation:Rotator)
    Scale="1,1,1"
   ) {
    rotationMode=BMM_Replace
    rotationSpace=BCS_WorldSpace
    scaleMode=BMM_Additive
    translationMode=BMM_Replace
    translationSpace=BCS_WorldSpace
    alphaBoolBlend {blendInTime=1, blendOption=Linear, blendOutTime=1}
    boneToModify {boneName=spine_05}
    rotation {roll=20}
   }
   n13 = Anim Dynamics - Bone: dyn_key_C!(ComponentPose=@n12, Alpha=@n2, GravityOverride:Vector="0,0,-980", ExternalForce="0,0,0") -> (Pose) {
    angularDampingOverride=0.9
    angularSpringConstant=250
    bAngularSpring=true
    bOverrideAngularDamping=true
    gravityScale=0
    alphaBoolBlend {blendOption=Linear}
    alphaScaleBiasClamp {bInterpResult=true}
    boundBone {boneName=dyn_key_C}
    componentAppliedLinearAccClamp=<2000 2000 2000>
    componentLinearAccScale=<0.5 0.5 0.5>
    componentLinearVelScale=<0.5 0.5 0.5>
    physicsBodyDefinitions [{
      boundBone {boneName=dyn_key_C}
      boxExtents=<5 10 10>
      constraintSetup {angularLimitsMax=<15 15 2>, angularLimitsMin=<-15 -15 -2>, angularTarget=<1 0 0>}
      localJointOffset=<1.5 1.5 0.5>
     }]
    relativeSpaceBone {boneName=root}
   }
   n14 = Anim Dynamics - Bone: dyn_key_B!(ComponentPose=@n13, Alpha=@n2, GravityOverride:Vector="0,0,0", ExternalForce="0,0,0") -> (Pose) {
    angularDampingOverride=0.9
    angularSpringConstant=50
    bAngularSpring=true
    bOverrideAngularDamping=true
    gravityScale=0
    alphaBoolBlend {blendOption=Linear}
    alphaScaleBiasClamp {bInterpResult=true}
    boundBone {boneName=dyn_key_B}
    componentAppliedLinearAccClamp=<2000 2000 2000>
    componentLinearAccScale=<0.5 0.5 0.5>
    componentLinearVelScale=<0.5 0.5 0.5>
    physicsBodyDefinitions [{
      boundBone {boneName=dyn_key_B}
      boxExtents=<12 12 7>
      constraintSetup {angularLimitsMax=<20 20 2>, angularLimitsMin=<-20 -20 -2>, angularTarget=<1 0 0>}
      localJointOffset=<2.5 0.5 2.5>
     }]
    relativeSpaceBone {boneName=root}
   }
   n15 = Anim Dynamics - Bone: dyn_key_A!(ComponentPose=@n14, Alpha=@n2, GravityOverride:Vector="0,0,0", ExternalForce="0,0,0") -> (Pose) {
    angularDampingOverride=0.9
    angularSpringConstant=50
    bAngularSpring=true
    bOverrideAngularDamping=true
    gravityScale=0
    alphaBoolBlend {blendOption=Linear}
    alphaScaleBiasClamp {bInterpResult=true}
    boundBone {boneName=dyn_key_A}
    componentAppliedLinearAccClamp=<2000 2000 2000>
    componentLinearAccScale=<0.5 0.5 0.5>
    componentLinearVelScale=<0.5 0.5 0.5>
    physicsBodyDefinitions [{
      boundBone {boneName=dyn_key_A}
      boxExtents=<10 10 5>
      constraintSetup {angularLimitsMax=<20 20 2>, angularLimitsMin=<-20 -20 -2>, angularTarget=<1 0 0>}
      localJointOffset=<1.5 1.5 0.5>
     }]
    relativeSpaceBone {boneName=root}
   }
   n16 = Anim Dynamics - Bone: dyn_granade!(ComponentPose=@n15, Alpha=@n2, GravityOverride:Vector="0,0,0", ExternalForce="0,0,0") -> (Pose) {
    angularDampingOverride=0.9
    angularSpringConstant=100
    bAngularSpring=true
    bOverrideAngularDamping=true
    gravityScale=0
    alphaBoolBlend {blendOption=Linear}
    alphaScaleBiasClamp {bInterpResult=true}
    boundBone {boneName=dyn_granade}
    componentAppliedLinearAccClamp=<2000 2000 2000>
    componentLinearAccScale=<0.5 0.5 0.5>
    componentLinearVelScale=<0.6 0.6 0.6>
    physicsBodyDefinitions [{
      boundBone {boneName=dyn_granade}
      boxExtents=<10 10 10>
      constraintSetup {angularLimitsMax=<10 5 2>, angularLimitsMin=<-10 -10 -5>, angularTarget=<1 0 0>}
      localJointOffset=<0.5 0.5 0>
     }]
    relativeSpaceBone {boneName=root}
   }
   n17 = Anim Dynamics - Bone: dyn_granade_clip!(ComponentPose=@n16, Alpha=@n2, GravityOverride:Vector="0,0,-1500", ExternalForce="0,0,0") -> (Pose) {
    angularDampingOverride=0.7
    angularSpringConstant=100
    bAngularSpring=true
    bOverrideAngularDamping=true
    bUseGravityOverride=true
    linearSpringConstant=100
    alphaBoolBlend {blendOption=Linear}
    alphaScaleBiasClamp {bInterpResult=true}
    boundBone {boneName=dyn_granade_clip}
    componentAppliedLinearAccClamp=<2000 2000 2000>
    componentLinearAccScale=<0.6 0.6 0.6>
    componentLinearVelScale=<0.6 0.6 0.6>
    gravityOverride {z=-50}
    physicsBodyDefinitions [{
      boundBone {boneName=dyn_granade_clip}
      boxExtents=<8 8 5>
      constraintSetup {angularLimitsMax=<10 10 35>, angularLimitsMin=<-10 -10 -100>, angularTarget=<1 0 0>}
      localJointOffset=<0.5 0 -1.5>
     }]
   }
   n18 = Anim Dynamics - Bone: dyn_buckle_A!(ComponentPose=@n17, Alpha=@n2, GravityOverride:Vector="0,0,-200", ExternalForce="0,0,0") -> (Pose) {
    angularDampingOverride=0.999
    angularSpringConstant=50
    bAngularSpring=true
    bOverrideAngularDamping=true
    bUseGravityOverride=true
    gravityScale=0
    alphaBoolBlend {blendOption=Linear}
    alphaScaleBiasClamp {bInterpResult=true}
    boundBone {boneName=dyn_buckle_A}
    componentAppliedLinearAccClamp=<2000 2000 2000>
    componentLinearAccScale=<0.2 0.2 0.2>
    componentLinearVelScale=<0.2 0.2 0.2>
    physicsBodyDefinitions [{
      boundBone {boneName=dyn_buckle_A}
      boxExtents=<3 10 10>
      constraintSetup {angularLimitsMax=<5 2 0>, angularLimitsMin=<-5 -2 -20>, angularTarget=<1 0 0>}
      localJointOffset=<0.5 -1.5 0.5>
     }]
    relativeSpaceBone {boneName=root}
   }
   n19 = Anim Dynamics - Bone: dyn_buckle_B!(ComponentPose=@n18, Alpha=@n2, GravityOverride:Vector="0,0,-350", ExternalForce="0,0,0") -> (Pose) {
    angularDampingOverride=0.99
    angularSpringConstant=200
    bAngularSpring=true
    bOverrideAngularDamping=true
    bUseGravityOverride=true
    gravityScale=0
    alphaBoolBlend {blendOption=Linear}
    alphaScaleBiasClamp {bInterpResult=true}
    boundBone {boneName=dyn_buckle_B}
    componentAppliedLinearAccClamp=<2000 2000 2000>
    componentLinearAccScale=<0.6 0.6 0.6>
    componentLinearVelScale=<0.6 0.6 0.6>
    physicsBodyDefinitions [{
      boundBone {boneName=dyn_buckle_B}
      boxExtents=<5 10 10>
      constraintSetup {angularLimitsMax=<15 5 2>, angularLimitsMin=<-15 -10 -2>, angularTarget=<1 0 0>}
      localJointOffset=<1 1 0>
     }]
    relativeSpaceBone {boneName=root}
   }
   n20 = Anim Dynamics - Bone: dyn_lower!(
    ComponentPose=@n19
    bAlphaBoolEnabled:bool=Get Always On() -> (Always On:bool)
    GravityOverride_X:double=0
    GravityOverride_Y:double=0
    GravityOverride_Z:double=Get Cloth sky dive corr() -> (Cloth sky dive corr:double)
    ExternalForce="0,0,0"
   ) -> (Pose) {
    alphaInputType=Bool
    angularDampingOverride=0.9999
    angularSpringConstant=50
    bAngularSpring=true
    bOverrideAngularDamping=true
    bUseGravityOverride=true
    bUsePlanarLimit=false
    gravityScale=0
    simulationSpace=BoneRelative
    alphaBoolBlend {blendInTime=0.3, blendOption=Linear, blendOutTime=0.4}
    alphaScaleBiasClamp {interpSpeedDecreasing=20, interpSpeedIncreasing=20}
    boundBone {boneName=dyn_lower}
    componentAppliedLinearAccClamp=<2000 2000 2000>
    componentLinearAccScale=<0.3 0.3>
    componentLinearVelScale=<0.4 0.4 0.6>
    physicsBodyDefinitions [{
      collisionType=OuterSphere
      sphereCollisionRadius=3.3
      boundBone {boneName=dyn_lower}
      boxExtents=<18 8 9>
      constraintSetup {angularLimitsMax=<2 20 2>, angularLimitsMin=<-2 -20 -50>, angularTarget=<1 0 0>}
      localJointOffset=<3.0497 -1.9971 0>
     }]
    planarLimits [{drivingBone {boneName=root}, planeTransform {location=<-2.7323 5.2074 -19.8655>, rotation=<rot 1.9444 -0.3187 -94.422>}}]
    relativeSpaceBone {boneName=root}
   }
   n21 = "Transform (Modify) Bone - Bone: dyn_lower_mstr"!(
    ComponentPose=@n20
    bAlphaBoolEnabled:bool=Get Cloth corr() -> (Cloth corr:bool)
    Translation="0,0,0"
    Rotation:Rotator="0,8,0"
    Scale="1,1,1"
   ) -> (Pose) {
    alphaInputType=Bool
    rotationMode=BMM_Additive
    rotationSpace=BCS_BoneSpace
    translationMode=BMM_Additive
    translationSpace=BCS_BoneSpace
    alphaBoolBlend {blendInTime=0.2, blendOption=Linear, blendOutTime=0.3}
    alphaScaleBiasClamp {inRange {max=0.3}}
    boneToModify {boneName=dyn_lower_mstr}
    rotation {yaw=8}
   }
   n22 = "Transform (Modify) Bone - Bone: pack"(ComponentPose=@n21, Alpha=@n1, Translation="0,-1,0", Rotation="0,0,12") {
    rotationMode=BMM_Additive
    rotationSpace=BCS_BoneSpace
    translationMode=BMM_Additive
    translationSpace=BCS_BoneSpace
    alphaBoolBlend {blendInTime=0.1, blendOption=Linear, blendOutTime=0.3}
    alphaScaleBiasClamp {bInterpResult=true, interpSpeedDecreasing=6, interpSpeedIncreasing=3}
    boneToModify {boneName=pack}
    rotation {roll=12}
    translation {y=-1}
   }
   n23 = "Transform (Modify) Bone - Bone: dyn_lower_mstr"!(ComponentPose=@n22, Alpha:float=@n1, Translation="0,1,0", Rotation="0,-15,0", Scale="1,1,1") -> (Pose) {
    rotationMode=BMM_Additive
    rotationSpace=BCS_BoneSpace
    translationMode=BMM_Additive
    translationSpace=BCS_BoneSpace
    alphaBoolBlend {blendInTime=0.2, blendOption=Linear, blendOutTime=0.3}
    alphaScaleBiasClamp {bInterpResult=true, inRange {max=0.3}}
    boneToModify {boneName=dyn_lower_mstr}
    rotation {yaw=-15}
    translation {y=1}
   }
   Output Pose(Result=Component To Local(ComponentPose=@n23)) {layerGroup=None}
  ]}
 EventGraph {nodes [
   n1 = Event BlueprintUpdateAnimation()
   > n2 = Sequence!(execute) -> (then_0, then_1, then_2:Exec)
   > Set Delta Time X(Delta Time X=@n1.DeltaTimeX)
   > Is Valid(InputObject=Get BodyMesh() -> (BodyMesh:Skeletal Mesh Component Object Reference))
   n3 = BreakTransform(InTransform=GetSocketTransform(self=Get BodyMesh(), InSocketName=pelvis, TransformSpace=RTS_World))
   > Set pelvis_translation(pelvis_translation:Vector=@n3) -> (Output_Get:Vector)
   > Set pelvis_rotation(pelvis_rotation:Rotator=@n3.Rotation) -> (Output_Get:Rotator)
   n4 = BreakTransform(InTransform=GetSocketTransform(self=Get BodyMesh(), InSocketName=spine_01, TransformSpace=RTS_World))
   > Set spine01_translation(spine01_translation:Vector=@n4) -> (Output_Get:Vector)
   > Set spine01_rotation(spine01_rotation:Rotator=@n4.Rotation) -> (Output_Get:Rotator)
   n5 = BreakTransform(InTransform=GetSocketTransform(self=Get BodyMesh(), InSocketName=spine_02, TransformSpace=RTS_World))
   > Set spine02_translation(spine02_translation:Vector=@n5) -> (Output_Get:Vector)
   > Set spine02_rotation(spine02_rotation:Rotator=@n5.Rotation) -> (Output_Get:Rotator)
   n6 = BreakTransform(InTransform=GetSocketTransform(self=Get BodyMesh(), InSocketName=spine_05, TransformSpace=RTS_World))
   > Set spine 05 translation(spine 05 translation:Vector=@n6) -> (Output_Get:Vector)
   > Set spine 05 rotation(spine 05 rotation:Rotator=@n6.Rotation) -> (Output_Get:Rotator)
   n7 = Get SkeletalMesh() -> (SkeletalMesh:Skeletal Mesh Component Object Reference)
   n8 = Get SkeletalMesh()
   n9 = "Clamp (Float)"(
    Value=float + float(
     A=GetAlphaFromReferenceTransformDelta(SkeletalMeshComponent=@n7, BoneName=spine_05, BaseName=pelvis, InputRange=(X=0,Y=-60), TwistAxis="1,0,0")
     B=GetAlphaFromReferenceTransformDelta(SkeletalMeshComponent=@n8, BoneName=spine_05, BaseName=pelvis, DeltaComponentType=EulerX, InputRange=(X=0,Y=40), TwistAxis="1,0,0")
    )
    Min=0
    Max=1
   )
   n10 = NOT Boolean(A=Get bIsCrouching())
   n11 = Get PawnMovementDirectionAzimuth() -> (PawnMovementDirectionAzimuth:float)
   n12 = AND Boolean(A="float > float"(A=@n11, B=-180), B="float < float"(A=@n11, B=-120))
   n13 = Get bIsMoving2D()
   > n14 = Set Alpha Output(Alpha Output:float=SelectFloat(
     A=MapRangeClamped(Value=@n9, InRangeA=0, InRangeB=1, OutRangeB=0.6)
     B=@n9
     bPickA=AND Boolean!(A=@n12, B=@n13, C:bool=@n10) -> (ReturnValue)
    )) -> (Output_Get:float)
   n15 = Get bIsJumping()
   n16 = Get bIsFalling()
   > "Set  jump alpha pack"(" jump alpha pack":double=SelectFloat(A=1, B=0, bPickA=OR Boolean(A=@n15, B=@n16))) -> (Output_Get:double)
   n17 = Get bIsJumping()
   n18 = Get bIsFalling()
   n19 = Get bIsSliding()
   > Set Wiggle alpha(Wiggle alpha:double=SelectFloat(A=0.6, B=@n14, bPickA=OR Boolean!(A=@n18, B=@n17, C:bool=@n19) -> (ReturnValue))) -> (Output_Get:double)
   n20 = Get bIsCrouching()
   n21 = Get PawnMovementDirectionAzimuth()
   n22 = "float > float"(A=@n21, B=-140)
   n23 = "float < float"(A=@n21, B=-25)
   > Set Is moving left(Is moving left:bool=AND Boolean(A=NOT Boolean(A=@n20), B=AND Boolean!(A=@n22, B=@n23) -> (ReturnValue), C=Get bIsMoving2D())) -> (Output_Get:bool)
   n24 = Get PawnMovementDirectionAzimuth()
   n25 = AND Boolean!(
    A=OR Boolean!(
     A="In Range (Float)"!(Value:double=@n24, Min:double=-95, Max:double=0, InclusiveMin:bool=true, InclusiveMax:bool=true) -> (ReturnValue:bool)
     B="In Range (Float)"(Value=@n24, Min=0, Max=95, InclusiveMin=true, InclusiveMax=true)
    ) -> (ReturnValue)
    B=Get bIsMoving2D()
    C:bool=OR Boolean(A=Get bIsJumping(), B=Get bIsFalling())
   ) -> (ReturnValue)
   n26 = OR Boolean(
    A="In Range (Float)"(Value=@n24, Min=-180, Max=-95, InclusiveMin=true, InclusiveMax=false)
    B="In Range (Float)"(Value=@n24, Min=95, Max=180, InclusiveMin=false, InclusiveMax=true)
   )
   n27 = OR Boolean(A=Get bIsJumping(), B=Get bIsFalling())
   n28 = Get bIsMoving2D()
   n29 = Get bIsMoving2D()
   n30 = Get bIsFalling()
   n31 = Get bIsJumping()
   n32 = SelectFloat(
    A=0.8
    B=SelectFloat(A=0.7, B=SelectFloat(A=1, B=0, bPickA=@n25), bPickA=AND Boolean(A=@n26, B=@n28, C=@n27))
    bPickA=AND Boolean!(A=NOT Boolean(A=@n29), B=OR Boolean(A=@n31, B=@n30)) -> (ReturnValue)
   )
   n33 = Get bIsSkyDiving()
   n34 = Get bIsSkydiveFloating()
   n35 = Get bIsParachuteOpened()
   > Set Jump alpha cloth(Jump alpha cloth:double=SelectFloat(A=0.35, B=SelectFloat(A=0.6, B=@n32, bPickA=@n33), bPickA=OR Boolean(A=@n34, B=@n35))) -> (Output_Get:double)
   n36 = Get bIsParachuteOpened()
   n37 = Get bIsSkydiveFloating()
   n38 = Get bIsSkyDiving()
   > Set Jump cloth Z rot corr(Jump cloth Z rot corr:double=SelectFloat(A=-20, B=-80, bPickA=OR Boolean!(A=@n37, B=@n36, C:bool=@n38) -> (ReturnValue))) -> (Output_Get:double)
   n39 = NOT Boolean(A=Get bIsParachuteOpened())
   n40 = Get bIsSkyDiving()
   > Set Cloth sky dive corr(Cloth sky dive corr:double=SelectFloat(A=1000, B=-2500, bPickA=AND Boolean(A=@n40, B=@n39))) -> (Output_Get:double)
   n41 = Get bIsTacticalSprinting()
   n42 = Get Alpha Output() -> (Alpha Output:float)
   Set Pack alpha(execute=@n2.then_1, Pack alpha:double=SelectFloat(A=0.3, B=SelectFloat(A=0.6, B=@n42, bPickA=@n41), bPickA=Get bIsPlayingMeleeAnim())) -> (Output_Get:double)
   n43 = Get bIsCrouching()
   n44 = Get PawnMovementDirectionAzimuth()
   n45 = "float > float"(A=@n44, B=-45)
   n46 = "float < float"(A=@n44, B=45)
   > Set Is moving forward(Is moving forward:bool=AND Boolean!(A=AND Boolean!(A=@n45, B=@n46) -> (ReturnValue), B=Get bIsMoving2D(), C:bool=NOT Boolean(A=@n43)) -> (ReturnValue)) -> (Output_Get:bool)
   n47 = Get bIsMoving2D()
   > Set Always On(Always On:bool=OR Boolean!(A=Get bIsMoving2D(), B=NOT Boolean(A=@n47)) -> (ReturnValue)) -> (Output_Get:bool)
   n48 = Get bIsCrouching()
   n49 = Get PawnMovementDirectionAzimuth()
   n50 = "float > float"(A=@n49, B=25)
   n51 = "float < float"(A=@n49, B=140)
   > n52 = Set Right mvmnt corr(Right mvmnt corr:bool=AND Boolean!(A=AND Boolean!(A=@n50, B=@n51) -> (ReturnValue), B=Get bIsMoving2D(), C:bool=NOT Boolean(A=@n48)) -> (ReturnValue)) -> (Output_Get:bool)
   n53 = Get bIsPlayingMeleeAnim()
   > Set Right movmnt and melee corr(Right movmnt and melee corr:double="To Float (Boolean)"(InBool=OR Boolean(A=@n53, B=@n52))) -> (Output_Get:double)
   n54 = Get bIsPlayingMeleeAnim()
   > Set Cloth corr(Cloth corr:bool=AND Boolean(A=NOT Boolean(A=@n54), B=Get Is moving forward() -> (Is moving forward:bool))) -> (Output_Get:bool)
   > Set Cloth alpha(Cloth alpha:double=MapRangeClamped(Value=Get Alpha Output(), InRangeA=0, InRangeB=1, OutRangeB=0.4)) -> (Output_Get:double)
   n55 = NOT Boolean(A=Get bIsSkyDiving())
   n56 = SelectFloat(A=-50, B=0, bPickA=AND Boolean(A=@n55, B=Get Is moving left() -> (Is moving left:bool)))
   n57 = Get bIsParachuteOpened()
   n58 = Get Right mvmnt corr() -> (Right mvmnt corr:bool)
   > Set Jump cloth Y rot corr(Jump cloth Y rot corr:double=SelectFloat(A=50, B=SelectFloat(A=0, B=@n56, bPickA=@n57), bPickA=AND Boolean(A=@n58, B=@n55))) -> (Output_Get:double)
   n59 = Branch(execute=@n2.then_2, Condition=Get Right mvmnt corr())
   > Do Once(Reset=@n60.then)
   > Set Start moving right(Start moving right:double=1) -> (Output_Get:double)
   > Delay(Duration=0.3)
   > Set Start moving right(Start moving right=0)
   n60 = Set Start moving right(execute=@n59.else, Start moving right=0)
   Event OnCharacterCustomizationLoaded()
   > n61 = Cast To PlayerPawn_Athena(Object=TryGetPawnOwner())
   > Set As Player Pawn Athena(As Player Pawn Athena:Player Pawn Athena Object Reference=@n61) -> (Output_Get:Player Pawn Athena Object Reference)
   > Set BodyMesh(BodyMesh=GetSkeletalMeshForPartType(self=@n61))
   > Set SkeletalMesh(SkeletalMesh:Skeletal Mesh Component Object Reference=GetPartSkeletalMeshComponent()) -> (Output_Get:Skeletal Mesh Component Object Reference)
  ]}
]
variables [
 " jump alpha pack" {type=number}
 Alpha Output {type=number}
 Always On {type=boolean}
 As Player Pawn Athena {type=PlayerPawn_Athena_C}
 BodyMesh {type=SkeletalMeshComponent}
 Cloth alpha {type=number}
 Cloth corr {type=boolean}
 Cloth sky dive corr {type=number}
 Delta Time X {type=number}
 Is moving forward {type=boolean}
 Is moving left {type=boolean}
 Jump alpha cloth {type=number}
 Jump cloth Y rot corr {type=number}
 Jump cloth Z rot corr {type=number}
 Pack alpha {type=number}
 Pack right mvmnt corr {type=boolean}
 pelvis_rotation {type=Rotator}
 pelvis_translation {type=Vector}
 Right movmnt and melee corr {type=number}
 Right mvmnt corr {type=boolean}
 SkeletalMesh {type=SkeletalMeshComponent}
 spine 05 rotation {type=Rotator}
 spine 05 translation {type=Vector}
 spine01_rotation {type=Rotator}
 spine01_translation {type=Vector}
 spine02_rotation {type=Rotator}
 spine02_translation {type=Vector}
 Start moving right {type=number}
 Wiggle alpha {type=number}
]
